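# Animation Blueprint for the face accessory of the male Padded Armor Order character, posing its hose and back tube attachment.
# It layers the body mesh's dyn_metalbox branch onto the copied parent pose, then runs a Control Rig and a spine_05/dyn_metalbox constraint.
# Head rotation relative to spine_05 drives an additive hose offset, followed by spring-based AnimDynamics.
# The head-back backTubeAttach_c transforms are not connected to the output pose.
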AnimBlueprint /BRCosmetics/Characters/Player/Male/Medium/Bodies/M_Med_Padded_Armor_Order/Meshes/Parts/M_MED_Padded_Armor_Order_FaceAcc_AnimBP
target_skeleton=/Game/Characters/Player/Male/Medium/Bodies/M_Med_Padded_Armor_Order/Meshes/Parts/M_MED_Padded_Armor_Order_FaceAcc_Skeleton
graphs [
 AnimGraph {nodes [
   Copy Pose From Mesh() {bUseAttachedParent=true}
   n1 = "Transform (Modify) Bone - Bone: backTubeAttach_c"(Alpha=Get head_rot_back() -> (head_rot_back:double), Translation="0,-4,-3.5", Rotation="0,0,0", Scale="1,1,1") {
    rotationMode=BMM_Additive
    rotationSpace=BCS_BoneSpace
    scaleMode=BMM_Additive
    scaleSpace=BCS_BoneSpace
    translationMode=BMM_Additive
    translationSpace=BCS_ParentBoneSpace
    alphaBoolBlend {blendOption=Linear}
    alphaScaleBiasClamp {interpSpeedDecreasing=25, interpSpeedIncreasing=25}
    boneToModify {boneName=backTubeAttach_c}
    translation {y=-4, z=-3.5}
   }
   "Transform (Modify) Bone - Bone: backTubeAttach_c"(ComponentPose=@n1, Alpha=Get head_rot_back_2() -> (head_rot_back_2:double), Translation="0,-1,-3", Rotation="0,0,0", Scale="1,1,1") {
    rotationMode=BMM_Additive
    rotationSpace=BCS_BoneSpace
    scaleMode=BMM_Additive
    scaleSpace=BCS_BoneSpace
    translationMode=BMM_Additive
    translationSpace=BCS_ParentBoneSpace
    alphaBoolBlend {blendOption=Linear}
    alphaScaleBiasClamp {interpSpeedDecreasing=25, interpSpeedIncreasing=25}
    boneToModify {boneName=backTubeAttach_c}
    translation {y=-1, z=-3}
   }
   n2 = Control Rig(Source=Use cached pose 'Start_Cache'()) {
    alphaBoolBlend {blendOption=Linear}
    controlRigAssetReference {blueprintRigClass=&/M_MED_Padded_Armor_Order_FaceAcc_CtrlRig.M_MED_Padded_Armor_Order_FaceAcc_CtrlRig_C}
   }
   n3 = "Constraint - backTubeAttach_c_zero (PP)"!(ComponentPose=Local To Component(LocalPose=@n2), Alpha=1, ConstraintWeights_0=0.47, ConstraintWeights_1:float=0.53) -> (Pose) {
    alphaBoolBlend {blendOption=Linear}
    boneToModify {boneName=backTubeAttach_c_zero}
    constraintSetup [{targetBone {boneName=spine_05}}, {targetBone {boneName=dyn_metalbox}}]
    constraintWeights [1,1]
   }
   n4 = "Transform (Modify) Bone - Bone: hose"(ComponentPose=@n3, Alpha=Get head_rot_left() -> (head_rot_left:double), Translation="0,-3,0", Rotation="-10,-35,10", Scale="1,1,1") {
    rotationMode=BMM_Additive
    rotationSpace=BCS_BoneSpace
    scaleMode=BMM_Additive
    scaleSpace=BCS_BoneSpace
    translationMode=BMM_Additive
    translationSpace=BCS_ParentBoneSpace
    alphaBoolBlend {blendOption=Linear}
    alphaScaleBiasClamp {interpSpeedDecreasing=25, interpSpeedIncreasing=25}
    boneToModify {boneName=hose}
    rotation=<rot -10 -35 10>
    translation {y=-3}
   }
   n5 = Anim Dynamics - Bone: hose(ComponentPose=@n4, Alpha=0.4, ExternalForce="0,0,0") {
    angularDampingOverride=1
    angularSpringConstant=1500
    bAngularSpring=true
    bGravityOverrideInSimSpace=true
    bOverrideAngularDamping=true
    gravityScale=0.35
    simulationSpace=RootRelative
    alphaBoolBlend {blendOption=Linear}
    boundBone {boneName=hose}
    physicsBodyDefinitions [{
      sphereCollisionRadius=0
      boundBone {boneName=hose}
      boxExtents=<10 20 20>
      constraintSetup {angularLimitsMax=<5 10 90>, angularLimitsMin=<-5 -10 -60>, angularTarget=<1 0 0>}
      localJointOffset=<10 0 0>
     }]
   }
   Output Pose(Result=Component To Local(ComponentPose=@n5)) {layerGroup=None}
   n6 = Get Body_Mesh()
   Save cached pose 'Start_Cache'(Pose=Layered blend per bone(BasePose=Copy Pose From Mesh() {bCopyCurves=true, bUseAttachedParent=true}, BlendPoses_0=Copy Pose From Mesh(SourceMeshComponent=@n6)) {layerSetup [{branchFilters [{boneName=dyn_metalbox}]}]}) {cacheName=Start_Cache}
  ]}
 EventGraph {nodes [
   Event BlueprintUpdateAnimation()
   > n1 = Cast To PlayerPawn_Athena(Object=TryGetPawnOwner())
   > Set Body_Mesh(Body_Mesh=GetSkeletalMeshForPartType(self=@n1))
   > Set Pawn Owner(Pawn Owner:Player Pawn Athena Object Reference) -> (Output_Get:Player Pawn Athena Object Reference)
   > Set Head_Mesh(Head_Mesh:Skeletal Mesh Component Object Reference) -> (Output_Get:Skeletal Mesh Component Object Reference)
   n2 = GetOwnerSkeletalMeshComponent()
   n3 = BreakRotator(InRot=BreakTransform(InTransform=GetDeltaTransformFromRefPose(self=@n2, BoneName=head, BaseName=spine_05)).Rotation)
   > Set head_rot_back(head_rot_back:double=MapRangeClamped(Value=@n3.Yaw, InRangeA=20, InRangeB=-40)) -> (Output_Get:double)
   > Set head_rot_left(head_rot_left:double=MapRangeClamped(Value=@n3, InRangeA=0, InRangeB=30)) -> (Output_Get:double)
   > Set head_rot_right(head_rot_right:double=MapRangeClamped(Value=@n3, InRangeA=0, InRangeB=-30)) -> (Output_Get:double)
   > Set head_rot_back_2(head_rot_back_2:double=MapRangeClamped(Value=@n3.Yaw, InRangeA=-50, InRangeB=-90)) -> (Output_Get:double)
   GetOwnerSkeletalMeshComponent()
   TryGetPawnOwner()
   Set Body_Mesh()
  ]}
]
variables [
 Body_Mesh {type=SkeletalMeshComponent}
 Head_Mesh {type=SkeletalMeshComponent}
 head_rot_back {type=number}
 head_rot_back_2 {type=number}
 head_rot_left {type=number}
 head_rot_right {type=number}
 Pawn Owner {type=PlayerPawn_Athena_C}
]
